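; Version 3 of the Sneaky Snowman player disguise, a child class that adds a frost explosion on death.
; Damage and fake-snowman interactions record a DamageCauser; death runs a one-time sphere overlap using scalable-float radius and launch speed.
; Characters launch outward, except the stored player pawn launches away from a valid DamageCauser, and optionally gain icy feet; vehicles get low friction.
; Using a fake swaps its mesh material, reusing the v2 asset.
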(event EventOnDamageServer
  (params (Damage :type Float%20%28single-precision%29) (DamageTags :type Gameplay%20Tag%20Container%20Structure%20%28by%20ref%29) (Momentum :type Vector :default "0, 0, 0") (HitInfo :type Hit%20Result%20Structure%20%28by%20ref%29) (InstigatedBy :type Controller%20Object%20Reference) (DamageCauser :type Actor%20Object%20Reference) (EffectContext :type Gameplay%20Effect%20Context%20Handle%20Structure))
  (Variables|Default|SetDamageCauser (Actor|GetOwner DamageCauser))
  (|CallParentFunction Damage DamageTags Momentum HitInfo InstigatedBy DamageCauser EffectContext))

(event EventOnDeathServer
  (params (Damage :type Float%20%28single-precision%29) (DamageTags :type Gameplay%20Tag%20Container%20Structure%20%28by%20ref%29) (Momentum :type Vector :default "0, 0, 0") (HitInfo :type Hit%20Result%20Structure%20%28by%20ref%29) (InstigatedBy :type Controller%20Object%20Reference) (DamageCauser :type Actor%20Object%20Reference) (EffectContext :type Gameplay%20Effect%20Context%20Handle%20Structure))
  (bind _damagecauser (Variables|Default|GetDamageCauser))
  (bind _returnvalue (Ability|GetAbilitySystemComponent _damagecauser))
  (bind _returnvalue_2 (Utilities|IsValid _returnvalue :!class /Script/Engine.KismetSystemLibrary))
  (|CallParentFunction Damage DamageTags Momentum HitInfo InstigatedBy 0 EffectContext)
  (Utilities|IsValid (Variables|Default|GetDamageCauser)
    (:Is%20Valid
      _)
    (:Is%20Not%20Valid
      (if (Math|Boolean|ANDBoolean (GameplayTags|TagContainer|CompareTagContainertoOtherTags (Variables|Default|GetT_InstantDeath) DamageTags) (Utilities|IsValid DamageCauser :!class /Script/Engine.KismetSystemLibrary) (!= DamageCauser self))
        _)))
  (if (not (Variables|Default|GetDead))
    (Variables|Default|SetDead true)
    (bind _outactors (Collision|SphereOverlapActors (Transformation|GetActorLocation) (ScalableFloat|GetValueatLevel (Variables|Default|GetFrostExplosionRadius)) (Utilities|Array|MakeArray "ObjectTypeQuery3" "ObjectTypeQuery5")))
    (for _array_element _outactors
      (Utilities|IsValid _array_element
        (:Is%20Valid
          (bind _ascharacter (Utilities|Casting|CastToCharacter _array_element)
            (:then
              (Character|LaunchCharacter _ascharacter (* (Math|Vector|Normalize (- (Transformation|GetActorLocation _array_element) (Transformation|GetActorLocation (select (and (Utilities|Equal(Object) _array_element (Variables|StoredActors|GetPlayerPawn)) (Utilities|IsValid (Variables|Default|GetDamageCauser) :!class /Script/Engine.KismetSystemLibrary)) (Variables|Default|GetDamageCauser) self)))) (ScalableFloat|GetValueatLevel (Variables|Default|GetLaunchSpeed))) true false)
              (Utilities|IsValid (Ability|GetAbilitySystemComponent _array_element)
                (:Is%20Valid
                  (if (ScalableFloat|GetValueasBool (Variables|Default|GetIcyFeetEnabled))
                    (GameplayEffects|ApplyGameplayEffectToTarget :self (select _returnvalue_2 _returnvalue (Ability|GetAbilitySystemComponent _array_element)) :GameplayEffectClass "/Game/Athena/Items/Consumables/SneakySnowman/Version3/GE_SnowmanV3_LowFriction_Athena.GE_SnowmanV3_LowFriction_Athena_C" :Target (Ability|GetAbilitySystemComponent _array_element) :Level 0.0)))))
            (:CastFailed
              (Utilities|Casting|CastToFortVehicleInterface _array_element)
              (GameplayEffects|ApplyGameplayEffectToTarget :self (select _returnvalue_2 _returnvalue (Ability|GetAbilitySystemComponent _array_element)) :GameplayEffectClass "/Game/Athena/Items/Consumables/ThrownConsumables/IceGrenade/GE_IceGrenade_LowFriction_Vehicle_Athena.GE_IceGrenade_LowFriction_Vehicle_Athena_C" :Target (Ability|GetAbilitySystemComponent _array_element) :Level 0.0))))))))

(event EventOnInteractWithFake
  (params (InteractingPawn :type Fort%20Pawn%20Object%20Reference))
  (Variables|Default|SetDamageCauser InteractingPawn)
  (|CallParentFunction InteractingPawn))

(fn OnStartedUsingFake
  ;[1 1 1] We want to be able to mat swap the fake from v2 without duplicating the fake snowman asset, this is for that purpose!
  (bind _asbuilding_gameplay_actor (Utilities|Casting|CastToBuildingGameplayActor (Variables|StoredActors|GetFakeActor))
    (:then
      (Utilities|IsValid (Actor|GetComponentbyClass _asbuilding_gameplay_actor "/Script/Engine.StaticMeshComponent")
        (:Is%20Valid
          (Rendering|Material|SetMaterial (Actor|GetComponentbyClass _asbuilding_gameplay_actor "/Script/Engine.StaticMeshComponent") 0 (Variables|Default|GetFakeSnowmanMaterial))
          (|CallParentFunction))
        (:Is%20Not%20Valid
          (|CallParentFunction))))
    (:CastFailed
      (|CallParentFunction))))

(fn ConstructionScript
  (|CallParentFunction))
