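; Sneaky Snowman V3 prop for Athena, which sets off a frost burst on death and destroys itself once playlist data is ready if disabled.
; On its first server death from weapon or vehicle-tagged damage, it sphere-overlaps nearby actors and launches characters away.
; It applies a low-friction "icy feet" effect to characters when enabled, and the Ice Grenade's vehicle low-friction effect to non-character vehicles.
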(event EventBeginPlay
  (|CallParentFunction)
  (Class|FortAsyncActionCurrentPlaylistReady|PlaylistDataReadyAsync
    (:OnReady
      (if (not (Class|AthenaFunctionLibrary|RowtoBool (Variables|Default|GetEnabled)))
        (Actor|DestroyActor)))))

(event EventOnDeathServer
  (params (Damage :type Float%20%28single-precision%29) (DamageTags :type Gameplay%20Tag%20Container%20Structure%20%28by%20ref%29) (Momentum :type Vector :default "0, 0, 0") (HitInfo :type Hit%20Result%20Structure%20%28by%20ref%29) (InstigatedBy :type Controller%20Object%20Reference) (DamageCauser :type Actor%20Object%20Reference) (EffectContext :type Gameplay%20Effect%20Context%20Handle%20Structure))
  (bind (_case_0_true _case_1_true) (GameplayTags|TagContainer|CompareTagContainertoOtherTags (Variables|Default|GetWeaponTag) (Variables|Default|GetVehicleTag) DamageTags))
  (bind _self self)
  (bind _returnvalue (Ability|GetAbilitySystemComponent _self))
  ;[0.01329 0.02713 0.15 0.5] On Death:
  (if (not (Variables|Default|GetDead))
    (Variables|Default|SetDead true)
    ;;[0.03193 0.15 0.02589 0.5] Old Version (v3) - Icy feet On Death
    (if (or _case_0_true _case_1_true)
      (bind _outactors (Collision|SphereOverlapActors (Transformation|GetActorLocation) (ScalableFloat|GetValueatLevel (Variables|Default|GetFrostExplosionRadius)) (Utilities|Array|MakeArray "ObjectTypeQuery3" "ObjectTypeQuery5")))
      (for _array_element _outactors
        (bind _ascharacter (Utilities|Casting|CastToCharacter _array_element)
          (:then
            (Character|LaunchCharacter _ascharacter (* (Math|Vector|Normalize (- (Transformation|GetActorLocation _array_element) (Transformation|GetActorLocation))) (ScalableFloat|GetValueatLevel (Variables|Default|GetLaunchSpeed))) true false)
            (if (ScalableFloat|GetValueasBool (Variables|Default|GetIcyFeetEnabled))
              (GameplayEffects|ApplyGameplayEffectToTarget :self _returnvalue :GameplayEffectClass "/Game/Athena/Items/Consumables/SneakySnowman/Version3/GE_SnowmanV3_LowFriction_Athena.GE_SnowmanV3_LowFriction_Athena_C" :Target (Ability|GetAbilitySystemComponent _array_element) :Level 0.0)))
          (:CastFailed
            (Utilities|Casting|CastToFortVehicleInterface _array_element)
            (GameplayEffects|ApplyGameplayEffectToTarget :self _returnvalue :GameplayEffectClass "/Game/Athena/Items/Consumables/ThrownConsumables/IceGrenade/GE_IceGrenade_LowFriction_Vehicle_Athena.GE_IceGrenade_LowFriction_Vehicle_Athena_C" :Target (Ability|GetAbilitySystemComponent _array_element) :Level 0.0)))))))

(fn OnSpawned_NotOnPawn)

(fn ConstructionScript
  (|CallParentFunction))
