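; Gameplay ability for the Sneaky Snowman V3 item that attributes a pawn's fall to whoever caused it.
; On activation it runs generic setup, caches the player pawn and waits for the falling movement-mode tag to be added.
; It then sets the pawn's fall instigator with a Sneaky Snowmando falling-damage tag.
; The instigator is the granting effect causer's controller if valid, otherwise the pawn's own controller.
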(event Ability|EventActivateAbility
  (|CallParentFunction)
  (bind (_event_data _fort_player_pawn) (Setup|SetupGenericAbility false false)
    (:Output
      (Variables|Default|SetPlayerPawn _fort_player_pawn)
      (Ability|Tasks|WaitGameplayTagAdd "(TagName=\"MovementMode.Falling\")"
        (:Added
          (bind _asfort_player_pawn_athena (Utilities|Casting|CastToFortPlayerPawnAthena (Variables|Default|GetPlayerPawn)))
          (Class|FortPlayerPawnAthena|SetFallInstigator _asfort_player_pawn_athena (select (Utilities|IsValid (Fortnite|GetFortPlayerControllerfromActor (Ability|EffectContext|GetEffectCauser (Ability|GetGrantedbyEffectContext))) :!class /Script/Engine.KismetSystemLibrary) (Fortnite|GetFortPlayerControllerfromActor (Ability|EffectContext|GetEffectCauser (Ability|GetGrantedbyEffectContext))) (Fortnite|GetFortPlayerControllerfromActor _asfort_player_pawn_athena)) "(GameplayTags=((TagName=\"Gameplay.Damage.Environment.Falling.SneakySnowmando\")))"))))))

(event Ability|EventOnEndAbility
  (params (bWasCancelled :type Boolean)))
